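# Niagara system for rising campfire smoke, built from one attached-ribbon emitter with two ribbon renderers.
# Particles launch upward with random drift, fading sideways acceleration and increasing drag; lifetime is capped by emitter age.
# Two panning, distorting, eroding material setups tint the smoke from hot orange to white as it fades.
# Two extra position attributes pull particles horizontally toward the simulation origin by different amounts, giving wide and narrow column variants.
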NiagaraSystem /FNGameplayCues/Effects/Fort_Effects/World/Enviroment/Smoke/NS_Janus_Campfire_Smoke_Rising
dependencies {
 usedDataInterfaces [&NiagaraDataInterfaceCurve, &NiagaraDataInterfaceVectorCurve, &NiagaraDataInterfaceColorCurve]
 usedDynamicInputs []
 usedModules [&SystemState]
}
emitters {FN_AttachedRibbon_v2 {
  emitterUpdateScript [
   &EmitterState()
   &FN_Integrator()
   &FN_SpawnRibbon_v2(Spawn Mode=::NewEnumerator0, SpawnRate=5)
  ]
  particleSpawnScript [
   &FN_InitializeRibbon_v3(
    Lifetime:NiagaraFloat=/Niagara/DynamicInputs/Math/Min_Float(A:NiagaraFloat=6, B:NiagaraFloat=$Emitter.Age)
    Ribbon Width:NiagaraFloat=&FloatFromCurve(FloatCurve=DataInterface() {propertyValues {Curve {keys [<at 0 1>,<at 1 1.25>]}}}, Scale Curve=130)
   )
   &FN_FlipRibbonV()
   &AddVelocity(Velocity=&RandomRangeVector(Minimum=<-25 -25 256>, Maximum=<25 25 256>))
  ]
  particleUpdateScript [
   &AccelerationForce(Acceleration=&VectorFromCurve(
     VectorCurve=DataInterface() {propertyValues {
       XCurve {keys [{interpMode=RCIM_Cubic, leaveTangent=-2.8967, tangentMode=RCTM_Break, time=0, value=1}, <at 1 0>]}
       YCurve {keys [{interpMode=RCIM_Cubic, leaveTangent=-2.8967, tangentMode=RCTM_Break, time=0, value=1}, <at 1 0>]}
       ZCurve {keys [{interpMode=RCIM_Cubic, leaveTangent=-2.8967, tangentMode=RCTM_Break, time=0, value=1}, <at 1 0>]}
      }}
     Scale Curve=&RandomRangeVector(Minimum=<24 -50 0>, Maximum=<50 50 1>)
    ))
   &Drag(
    Drag=&FloatFromCurve(FloatCurve=DataInterface() {propertyValues {Curve {keys [
         {interpMode=RCIM_Cubic, tangentMode=RCTM_Break, time=0.1989, value=0}
         {arriveTangent=3.6394, interpMode=RCIM_Cubic, leaveTangent=3.6394, tangentMode=RCTM_User, time=1, value=1}
        ]}}})
    Use Linear Drag=true
    Use Rotational Drag=false
   )
   &SolveForcesAndVelocity(Clamp Velocity=false, Limit Acceleration=false)
   &FN_UpdateRibbon(
    Enable Skeletal Position=true
    Initial Color:LinearColor=$Particles.Initial.Color
    Initial RibbonWidth:NiagaraFloat=$Particles.Initial.RibbonWidth
   )
   /Game/Effects/Niagara/Modules/Materials/FN_RibbonMaterial/FN_RibbonMaterial(
    Color:LinearColor=&ScaleLinearColorByCurve(
     InitialValue=<rgba 1 1 1 0.275>
     ScaleCurve=DataInterface() {propertyValues {
       AlphaCurve {keys [<at 0.0274 1>,<at 0.7 0.4>,<at 0.9945 0>]}
       BlueCurve {keys [<at 0.0145 0>,<at 0.125 1>]}
       GreenCurve {keys [<at 0.0145 2.2113>,<at 0.125 1>]}
       RedCurve {keys [<at 0.0145 7>,<at 0.125 1>]}
      }}
     CurveIndex=$Particles.NormalizedAge
    )
    Color Pan U Speed=0.3
    Color Scale U=3
    Erosion Pan Speed U=0.5
    Erosion Power=4
    Erossion Scale U=0.25
    Macro Distortion Amount:NiagaraFloat=&FloatFromCurve(
     FloatCurve=DataInterface() {propertyValues {Curve {keys [{arriveTangent=5.049, interpMode=RCIM_Cubic, leaveTangent=5.049, tangentMode=RCTM_User, time=0, value=0}, <at 0.5019 1 RCIM_Cubic>]}}}
     Scale Curve=0.077
    )
    Macro Pan Speed=-0.15
    Macro Scale U=0.25
    Micro Distortion Amount:NiagaraFloat=&FloatFromCurve(
     FloatCurve=DataInterface() {propertyValues {Curve {keys [{arriveTangent=4.3448, interpMode=RCIM_Cubic, leaveTangent=4.3448, tangentMode=RCTM_User, time=0, value=0}, <at 0.5075 1 RCIM_Cubic>]}}}
     Scale Curve=0.05
    )
    Micro Pan Speed=0.2
    Opacity Pan U Speed=0.4
    Opacity Scale U=3
    TIme:NiagaraFloat=$Emitter.FN_Integrator.Value
    Write to Module Namespace=true
   )
   &FN_RibbonMaterial(
    Color:LinearColor=&ScaleLinearColorByCurve(
     InitialValue=<rgba 1 1 1 0.275>
     ScaleCurve=DataInterface() {propertyValues {
       AlphaCurve {keys [<at 0.0274 1>,<at 0.7 0.4>,<at 0.9969 0>]}
       BlueCurve {keys [<at 0.0145 0>,<at 0.125 1>]}
       GreenCurve {keys [<at 0.0145 2.2113>,<at 0.125 1>]}
       RedCurve {keys [<at 0.0145 7>,<at 0.125 1>]}
      }}
     CurveIndex=$Particles.NormalizedAge
    )
    Color Pan U Speed=0.75
    Color Scale U=4
    Erosion Pan Speed U=0.5
    Erosion Power=4
    Erossion Scale U=0.25
    Macro Distortion Amount:NiagaraFloat=&FloatFromCurve(
     FloatCurve=DataInterface() {propertyValues {Curve {keys [{arriveTangent=5.049, interpMode=RCIM_Cubic, leaveTangent=5.049, tangentMode=RCTM_User, time=0, value=0}, <at 0.5019 1 RCIM_Cubic>]}}}
     Scale Curve=0.077
    )
    Macro Pan Speed=-0.15
    Macro Scale U=0.25
    Micro Distortion Amount:NiagaraFloat=&FloatFromCurve(
     FloatCurve=DataInterface() {propertyValues {Curve {keys [{arriveTangent=4.3448, interpMode=RCIM_Cubic, leaveTangent=4.3448, tangentMode=RCTM_User, time=0, value=0}, <at 0.5075 1 RCIM_Cubic>]}}}
     Scale Curve=0.05
    )
    Micro Pan Speed=0.2
    Opacity Pan U Speed=0.75
    Opacity Scale U=4
    TIme:NiagaraFloat=$Emitter.FN_Integrator.Value
    Write to Module Namespace=true
   ) {module=FN_RibbonMaterial001}
   /SetVariables(Particles.ParticlesPositionScalePrimary:Vector3f=&Add_Vector(
     A=&Multiply_Vector(
      A=&Subtract_Vector(A=Linked() {linkedVariable {name=Particles.Position, type=NiagaraPosition}}, B=/Niagara/DynamicInputs/Helpers/SimulationPosition())
      B=&VectorFromCurve(VectorCurve=DataInterface() {propertyValues {
         XCurve {keys [{interpMode=RCIM_Cubic, leaveTangent=-2.5999, tangentMode=RCTM_Break, time=0, value=1}, <at 1 0.45>]}
         YCurve {keys [{interpMode=RCIM_Cubic, leaveTangent=-2.6159, tangentMode=RCTM_Break, time=0, value=1}, <at 1 0.45>]}
         ZCurve {keys [<at 0 1>,<at 1 1>]}
        }})
     )
     B=&SimulationPosition()
    )) {isSetParameters=true}
   /SetVariables(Particles.ParticlePositionScaleSecondary:Vector3f=&Add_Vector(
     A=&Multiply_Vector(
      A=&Subtract_Vector(A=Linked() {linkedVariable {name=Particles.Position, type=NiagaraPosition}}, B=&SimulationPosition())
      B=&VectorFromCurve(VectorCurve=DataInterface() {propertyValues {
         XCurve {keys [{interpMode=RCIM_Cubic, leaveTangent=-0.8153, tangentMode=RCTM_Break, time=0, value=0.34}, <at 1 0>]}
         YCurve {keys [{interpMode=RCIM_Cubic, leaveTangent=-0.8207, tangentMode=RCTM_Break, time=0, value=0.34}, <at 1 0>]}
         ZCurve {keys [<at 0 1>,<at 1 1>]}
        }})
     )
     B=&SimulationPosition()
    )) {isSetParameters=true}
   &FN_RibbonMaterial() {enabled=false, module=FN_RibbonMaterial003}
  ]
  rendererClasses [&NiagaraRibbonRendererProperties]
  renderers [
   {rendererClass=&NiagaraRibbonRendererProperties}
   {rendererClass=&NiagaraRibbonRendererProperties, rendererIndex=1}
  ]
 }}
summary {
 systemName=NS_Janus_Campfire_Smoke_Rising
 emitters [{emitterName=FN_AttachedRibbon_v2, rendererClasses [&NiagaraRibbonRendererProperties]}]
 userVariables [
  User.RibbonIDOffset {type=NiagaraInt32}
  User.SkeletalMesh {
   type=NiagaraDataInterfaceSkeletalMesh
   defaultValue {
    struct=NiagaraExt_VariableValue_DataInterface
    value {dataInterface=&self:NiagaraDataInterfaceSkeletalMesh_1}
   }
  }
 ]
}
